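# Sound cue for a button click in the front-end UI of the Season 24 Star Wars engagement event.
# A single wave player feeds the root directly, reusing the main front-end button click sound rather than an event-specific wave.
# It routes through the Fort UI sound class and overrides concurrency to cap simultaneous clicks, stop the oldest and space out rapid retriggers.
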SoundCue /Game/Sounds/Fort_UI_Sounds/FrontEnd/EngagementEvent/S24_StarWars/S24_EE_Button_Click_01_Cue
root=WavePlayer_2
nodes [{id=WavePlayer_2, properties {soundWaveAssetPtr=/Game/Sounds/Fort_UI_Sounds/FrontEnd/New_FrontEnd/UI_FN_Button_Click_Main_01}}]
properties {
 bOverrideConcurrency=true
 maxDistance=2097152
 soundClassObject=&Fort_UI
 volumeMultiplier=1
 concurrencyOverrides {maxCount=3, resolutionRule=StopOldest, retriggerTime=0.05}
}
