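# Sound cue for the upgrade button click in the Season 24 Star Wars front-end engagement event UI.
# A mixer layers a low whoosh, boosted through a fixed-volume modulator and slightly attenuated at the mixer, with the standard front-end button click at full level.
# It plays in the UI sound class with overridden concurrency that caps simultaneous copies, rejects new plays beyond the cap and sets a short retrigger time.
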SoundCue /Game/Sounds/Fort_UI_Sounds/FrontEnd/EngagementEvent/S24_StarWars/S24_EE_Button_Click_Upgrade_Cue
root=Mixer_0
nodes [
 {id=Mixer_0, children [Modulator_1,WavePlayer_1], properties {inputVolume [0.7,1]}}
 {id=Modulator_1, children [WavePlayer_0], properties {volumeMax=2.5, volumeMin=2.5}}
 {id=WavePlayer_0, properties {soundWaveAssetPtr=/Game/Sounds/Fort_UI_Sounds/FrontEnd/EngagementEvent/S24_StarWars/wav/S42_EE_Whoosh_Low_01}}
 {id=WavePlayer_1, properties {soundWaveAssetPtr=/Game/Sounds/Fort_UI_Sounds/FrontEnd/New_FrontEnd/UI_FN_Button_Click_Main_01}}
]
properties {
 bOverrideConcurrency=true
 maxDistance=2097152
 soundClassObject=&Fort_UI
 volumeMultiplier=1
 concurrencyOverrides {maxCount=3, resolutionRule=PreventNew, retriggerTime=0.2}
}
