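# Sound cue for the front-end UI that, by its name, plays when a reward purchase completes in the Season 24 Star Wars engagement event.
# A single wave player outputs the shared loud press-and-hold button completion sound at boosted volume on the Fort_UI sound class.
# A concurrency override allows one instance at a time, blocking new plays rather than cutting off the current one, with a short retrigger time.
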SoundCue /Game/Sounds/Fort_UI_Sounds/FrontEnd/EngagementEvent/S24_StarWars/S24_EE_BuyRewards_Complete_Cue
root=WavePlayer_3
nodes [{
  id=WavePlayer_3
  properties {soundWaveAssetPtr=/Game/Sounds/Fort_UI_Sounds/FrontEnd/DefaultSounds/HoldableButtons/wav/UI_Button_Press_Hold_Loud_Complete}
 }]
properties {
 bOverrideConcurrency=true
 maxDistance=2097152
 soundClassObject=&Fort_UI
 volumeMultiplier=1.65
 concurrencyOverrides {maxCount=1, resolutionRule=PreventNew, retriggerTime=0.2}
}
