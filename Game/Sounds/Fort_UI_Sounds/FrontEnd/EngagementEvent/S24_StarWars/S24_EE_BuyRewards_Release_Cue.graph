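# Sound cue for releasing the buy-rewards button in the Season 24 Star Wars engagement event front-end UI.
# A single wave player, with no mixing or randomization, plays the shared default holdable-button loud release sample at slightly reduced volume.
# It routes through the Fort_UI sound class and overrides concurrency to cap simultaneous plays, stopping the oldest.
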SoundCue /Game/Sounds/Fort_UI_Sounds/FrontEnd/EngagementEvent/S24_StarWars/S24_EE_BuyRewards_Release_Cue
root=WavePlayer_1
nodes [{
  id=WavePlayer_1
  properties {soundWaveAssetPtr=/Game/Sounds/Fort_UI_Sounds/FrontEnd/DefaultSounds/HoldableButtons/wav/UI_Button_Press_Hold_Loud_Release}
 }]
properties {
 bOverrideConcurrency=true
 maxDistance=2097152
 soundClassObject=&Fort_UI
 volumeMultiplier=0.85
 concurrencyOverrides {maxCount=10, resolutionRule=StopOldest}
}
