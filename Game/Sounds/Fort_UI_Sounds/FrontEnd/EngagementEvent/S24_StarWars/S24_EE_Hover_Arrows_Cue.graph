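# Sound cue named for hovering arrows in the Season 24 Star Wars engagement event front-end UI.
# A single wave player feeds the output, reusing the default front-end hover sound rather than event-specific audio, routed through the Fort_UI sound class.
# Its own concurrency override allows one instance at a time and blocks new plays, with a short retrigger time to limit rapid repeats.
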SoundCue /Game/Sounds/Fort_UI_Sounds/FrontEnd/EngagementEvent/S24_StarWars/S24_EE_Hover_Arrows_Cue
root=WavePlayer_1
nodes [{id=WavePlayer_1, properties {soundWaveAssetPtr=/Game/Sounds/Fort_UI_Sounds/FrontEnd/DefaultSounds/PressedButtons/wav/UI_MP_Hover_Default_01}}]
properties {
 bOverrideConcurrency=true
 maxDistance=2097152
 soundClassObject=&Fort_UI
 volumeMultiplier=1
 concurrencyOverrides {maxCount=1, resolutionRule=PreventNew, retriggerTime=0.2}
}
