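# Sound cue for hovering a tile in the Season 24 Star Wars engagement event front-end UI.
# A single wave player plays the shared default UI hover wave directly, with no mixing or modulation, on the Fort_UI sound class.
# Its own concurrency override limits stacking to two voices, stopping the oldest, with a short retrigger time and volume scaling for overlapping plays.
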SoundCue /Game/Sounds/Fort_UI_Sounds/FrontEnd/EngagementEvent/S24_StarWars/S24_EE_Hover_Tile_Cue
root=WavePlayer_2
nodes [{id=WavePlayer_2, properties {soundWaveAssetPtr=/Game/Sounds/Fort_UI_Sounds/FrontEnd/DefaultSounds/PressedButtons/wav/UI_MP_Hover_Default_01}}]
properties {
 bOverrideConcurrency=true
 maxDistance=2097152
 soundClassObject=&Fort_UI
 volumeMultiplier=1
 concurrencyOverrides {maxCount=2, resolutionRule=StopOldest, retriggerTime=0.1, volumeScale=0.3}
}
