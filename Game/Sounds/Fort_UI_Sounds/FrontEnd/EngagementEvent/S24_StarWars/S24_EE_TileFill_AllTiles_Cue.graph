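# UI sound cue named for the all-tiles fill moment in the Season 24 Star Wars front-end engagement event.
# A single wave player is the root, with no wave asset shown, routed to the Fort_UI sound class with a slight volume boost.
# Its own concurrency override allows one instance, rejecting new plays while it sounds, with a short retrigger time.
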SoundCue /Game/Sounds/Fort_UI_Sounds/FrontEnd/EngagementEvent/S24_StarWars/S24_EE_TileFill_AllTiles_Cue
root=WavePlayer_0
nodes [{id=WavePlayer_0}]
properties {
 bOverrideConcurrency=true
 maxDistance=2097152
 soundClassObject=&Fort_UI
 volumeMultiplier=1.2
 concurrencyOverrides {maxCount=1, resolutionRule=PreventNew, retriggerTime=0.05}
}
