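# Front-end UI sound cue for an individual tile fill in the Season 24 Star Wars engagement event, judging by its name and path.
# A single wave player feeds the output at reduced volume, routed through the Fort_UI sound class; no wave asset is shown.
# Overridden concurrency caps simultaneous instances, rejects new ones at the limit, and adds a retrigger time and volume scaling, restraining rapid repeated plays.
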SoundCue /Game/Sounds/Fort_UI_Sounds/FrontEnd/EngagementEvent/S24_StarWars/S24_EE_TileFill_Individual_Cue
root=WavePlayer_0
nodes [{id=WavePlayer_0}]
properties {
 bOverrideConcurrency=true
 maxDistance=2097152
 soundClassObject=&Fort_UI
 volumeMultiplier=0.5
 concurrencyOverrides {maxCount=2, resolutionRule=PreventNew, retriggerTime=0.2, volumeScale=0.4}
}
