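# Front-end UI sound cue, named for the hologram rez-in on the upsell transition in the Season 24 Star Wars engagement event.
# A single wave player reuses the confirm-window appear whoosh, raised in pitch and lowered in volume, under the Fort UI sound class.
# A concurrency override allows one instance at a time, blocking new plays rather than stopping the current one, with a short retrigger time.
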SoundCue /Game/Sounds/Fort_UI_Sounds/FrontEnd/EngagementEvent/S24_StarWars/S42_EE_Trans_Upsell_Hologram_RezIn_Cue
root=WavePlayer_3
nodes [{id=WavePlayer_3, properties {soundWaveAssetPtr=/Game/Sounds/Fort_UI_Sounds/FrontEnd/Whooshes/Toasts/wav/UI_Whoosh_ConfirmWindow_Appear}}]
properties {
 bOverrideConcurrency=true
 maxDistance=2097152
 pitchMultiplier=1.3
 soundClassObject=&Fort_UI
 volumeMultiplier=0.5
 concurrencyOverrides {maxCount=1, resolutionRule=PreventNew, retriggerTime=0.2}
}
